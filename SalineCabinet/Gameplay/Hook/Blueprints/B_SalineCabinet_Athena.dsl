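; Weapon Blueprint for the SalineCabinet hook item that preloads a glider asset when the weapon is equipped.
; BeginPlay, which runs on equip, calls the parent, then asynchronously changes the bundle state of the Glider_SalineCabinet_Grape primary asset to add its Equipped bundle.
; This early Asset Manager request is meant to have the glider in memory before it is used.
; The construction script only calls its parent.
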(event EventBeginPlay
  (|CallParentFunction)
  ;[1 1 1] This happens on equip of the weapon. We need to "preload" this asset so it'll be there when we use it.
  (AssetManager|AsyncChangeBundleStateForPrimaryAssetList (Utilities|Array|MakeArray "AthenaGlider:Glider_SalineCabinet_Grape") (Utilities|Array|MakeArray "Equipped") (Utilities|Array|MakeArray "None")))

(fn ConstructionScript
  (|CallParentFunction))
